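# Physics asset for the canteen accessory on the CoconutShell medium male player body cosmetic, giving it secondary motion.
# A three-link chain of dynamic canteen capsules hangs from a small master bone, joined by tightly limited swing and twist constraints that mostly bend along one swing axis.
# Pelvis and thigh capsules have no constraints and appear to serve as collision shapes for the swinging canteen.
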PhysicsAsset /BRCosmetics/Characters/Player/Male/Medium/Bodies/M_MED_CoconutShell/Meshes/M_MED_CoconutShell_Canteen_Physics
bodies {
 dyn_canteen_01 [{shapeType=Capsule, length=8.5732, radius=4, shapeName=dyn_canteen_01_capsule, center=<3.2596 0 3.7756>, rotation=<rot 0 0 90>}]
 dyn_canteen_02 [{shapeType=Capsule, length=5.7109, radius=4, shapeName=dyn_canteen_02_capsule, center=<5.4133 0 3.092>, rotation=<rot 0 0 90>}]
 dyn_canteen_03 [{shapeType=Capsule, length=6.8841, radius=4, shapeName=dyn_canteen_03_capsule, center=<5.4632 0 1.96>, rotation=<rot 0 0 90>}]
 dyn_canteen_MSTR [{shapeType=Capsule, length=0.505, radius=0.505, shapeName=dyn_canteen_MSTR_capsule}]
 pelvis [
  {shapeType=Capsule, length=10.4676, radius=7.7675, shapeName=pelvis_capsule, center=<-0.8839 -1.8272 0>, rotation=<rot 0 -90 0>}
  {shapeType=Capsule, length=11.9091, radius=6.858, shapeName=pelvis_capsule, center=<0.4252 0.3593 0>}
 ]
 thigh_l [{shapeType=Capsule, length=25.9663, radius=8.5916, shapeName=thigh_r_capsule, center=<-18.1557 0.0265 1.096>, rotation=<rot 90 180 0.0005>}]
 thigh_r [{
   shapeType=Capsule
   length=26.3959
   radius=8.162
   shapeName=thigh_r_capsule
   center=<18.1557 -0.0265 -1.0247>
   rotation=<rot 90 180 -179.9995>
  }]
}
constraints [
 {bone1Name=dyn_canteen_01, bone2Name=dyn_canteen_MSTR, swing1LimitDegrees=5, swing2LimitDegrees=10, twistLimitDegrees=4}
 {bone1Name=dyn_canteen_02, bone2Name=dyn_canteen_01, swing1LimitDegrees=5, swing2LimitDegrees=20, twistLimitDegrees=4}
 {bone1Name=dyn_canteen_03, bone2Name=dyn_canteen_02, swing1LimitDegrees=0, swing2LimitDegrees=20, twistLimitDegrees=4}
]
