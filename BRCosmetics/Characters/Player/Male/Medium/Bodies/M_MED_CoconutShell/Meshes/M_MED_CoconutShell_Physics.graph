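# Physics asset for the male medium CoconutShell body cosmetic, setting up secondary motion for a canteen and front and back cloth flaps.
# Each dangling part is a capsule chain hung from a MSTR bone with tight swing and twist limits.
# Pelvis, lower spine, thigh and calf capsules supply body shapes without constraints.
# Angular-free cross constraints tie left and right flaps together along the lower segments, coupling each pair while letting them rotate.
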PhysicsAsset /BRCosmetics/Characters/Player/Male/Medium/Bodies/M_MED_CoconutShell/Meshes/M_MED_CoconutShell_Physics
bodies {
 calf_l [{
   shapeType=Capsule
   length=44.5314
   radius=6.9559
   shapeName=calf_r_capsule
   center=<-17.5781 -2.0592 0.302>
   rotation=<rot -89.321 -0.0011 -179.263>
  }]
 calf_r [{
   shapeType=Capsule
   length=41.0061
   radius=6.9559
   shapeName=calf_r_capsule
   center=<17.9576 2.4041 -0.2975>
   rotation=<rot -89.321 -0.0005 0.0003>
  }]
 dyn_canteen_01 [{shapeType=Capsule, length=8.5732, radius=4, shapeName=dyn_canteen_01_capsule, center=<3.2596 0 3.7756>, rotation=<rot 0 0 90>}]
 dyn_canteen_02 [{shapeType=Capsule, length=5.7109, radius=4, shapeName=dyn_canteen_02_capsule, center=<5.4133 0 3.092>, rotation=<rot 0 0 90>}]
 dyn_canteen_03 [{shapeType=Capsule, length=6.8841, radius=4, shapeName=dyn_canteen_03_capsule, center=<5.4632 0 1.96>, rotation=<rot 0 0 90>}]
 dyn_canteen_MSTR [{shapeType=Capsule, length=0.505, radius=0.505, shapeName=dyn_canteen_MSTR_capsule}]
 dyn_cloth_bk_l_01 [{shapeType=Capsule, length=9, radius=6, shapeName=dyn_cloth_back_l_02_capsule, center=<0 0 4.6019>, rotation=<rot 0 0 90>}]
 dyn_cloth_bk_l_02 [{shapeType=Capsule, length=9, radius=6, shapeName=dyn_cloth_back_l_03_capsule, center=<0 0 3.4159>, rotation=<rot 0 0 90>}]
 dyn_cloth_bk_l_03 [{shapeType=Capsule, length=9, radius=6, shapeName=dyn_cloth_back_l_04_capsule, center=<0.381 0 3>, rotation=<rot 0 0 90>}]
 dyn_cloth_bk_l_04 [{shapeType=Capsule, length=8.4416, radius=6, shapeName=dyn_cloth_back_l_04_capsule, center=<0.7077 0 3>, rotation=<rot 0 0 90>}]
 dyn_cloth_bk_l_MSTR [{shapeType=Capsule, length=0.505, radius=0.505, shapeName=dyn_cloth_back_l_master_capsule, rotation=<rot 0 -90 0>}]
 dyn_cloth_bk_r_01 [{shapeType=Capsule, length=9, radius=6, shapeName=dyn_cloth_back_l_02_capsule, center=<0 0 4.6019>, rotation=<rot 0 0 90>}]
 dyn_cloth_bk_r_02 [{shapeType=Capsule, length=9, radius=6, shapeName=dyn_cloth_back_l_03_capsule, center=<0 0 3.4159>, rotation=<rot 0 0 90>}]
 dyn_cloth_bk_r_03 [{shapeType=Capsule, length=9, radius=6, shapeName=dyn_cloth_back_l_04_capsule, center=<0.381 0 3>, rotation=<rot 0 0 90>}]
 dyn_cloth_bk_r_04 [{shapeType=Capsule, length=8.4416, radius=6, shapeName=dyn_cloth_back_l_04_capsule, center=<0.7077 0 3>, rotation=<rot 0 0 90>}]
 dyn_cloth_bk_r_MSTR [{shapeType=Capsule, length=0.505, radius=0.505, shapeName=dyn_cloth_back_l_master_capsule, rotation=<rot 0 -90 0>}]
 dyn_cloth_fr_l_01 [{
   shapeType=Capsule
   length=7.7014
   radius=4.9208
   shapeName=dyn_cloth_front_l_01_capsule
   center=<6.4126 0.9091 2.3>
   rotation=<rot -90 180 180>
  }]
 dyn_cloth_fr_l_02 [{
   shapeType=Capsule
   length=9.2005
   radius=5.1817
   shapeName=dyn_cloth_front_l_02_capsule
   center=<7.0504 0.7042 2.6369>
   rotation=<rot -89.8632 -179.9999 -180>
  }]
 dyn_cloth_fr_l_03 [{
   shapeType=Capsule
   length=8.8233
   radius=5.0623
   shapeName=dyn_cloth_front_l_03_capsule
   center=<8.0558 0.471 2.785>
   rotation=<rot -89.5187 0.0002 0>
  }]
 dyn_cloth_fr_l_04 [{
   shapeType=Capsule
   length=5.3546
   radius=4.0295
   shapeName=dyn_cloth_front_l_04_capsule
   center=<6.9693 0.1837 1.6269>
   rotation=<rot -90 0 0>
  }]
 dyn_cloth_fr_l_MSTR [{shapeType=Capsule, length=0.505, radius=0.505, shapeName=dyn_cloth_front_l_master_capsule, rotation=<rot 0 -90 0>}]
 dyn_cloth_fr_r_01 [{
   shapeType=Capsule
   length=7.2455
   radius=4.8849
   shapeName=dyn_cloth_front_l_01_capsule
   center=<6.0286 -0.7836 2.3>
   rotation=<rot -90 0 180>
  }]
 dyn_cloth_fr_r_02 [{
   shapeType=Capsule
   length=9.2005
   radius=5.1817
   shapeName=dyn_cloth_front_l_02_capsule
   center=<6.9768 -0.9136 2.637>
   rotation=<rot -89.8632 -179.9999 0>
  }]
 dyn_cloth_fr_r_03 [{
   shapeType=Capsule
   length=8.8233
   radius=5.0623
   shapeName=dyn_cloth_front_l_03_capsule
   center=<7.4044 -1.2132 2.785>
   rotation=<rot -89.5187 0.0002 -180>
  }]
 dyn_cloth_fr_r_04 [{
   shapeType=Capsule
   length=5.3546
   radius=4.0295
   shapeName=dyn_cloth_front_l_04_capsule
   center=<6.7954 -0.156 1.6269>
   rotation=<rot -90 180 0>
  }]
 dyn_cloth_fr_r_MSTR [{shapeType=Capsule, length=0.505, radius=0.505, shapeName=dyn_cloth_front_r_master_capsule, rotation=<rot 0 -90 0>}]
 pelvis [
  {shapeType=Capsule, length=10.4676, radius=7.7675, shapeName=pelvis_capsule, center=<-0.8839 -1.8272 0>, rotation=<rot 0 -90 0>}
  {shapeType=Capsule, length=11.9091, radius=6.858, shapeName=pelvis_capsule, center=<0.4252 0.3593 0>}
 ]
 spine_01 [{shapeType=Capsule, length=10.4884, radius=6.7639, shapeName=spine_01_capsule, rotation=<rot 0 -90 0>}]
 thigh_l [{shapeType=Capsule, length=25.9663, radius=8.5916, shapeName=thigh_r_capsule, center=<-18.1557 0.0265 1.096>, rotation=<rot 90 180 0.0005>}]
 thigh_r [{
   shapeType=Capsule
   length=26.3959
   radius=8.162
   shapeName=thigh_r_capsule
   center=<18.1557 -0.0265 -1.0247>
   rotation=<rot 90 180 -179.9995>
  }]
}
constraints [
 {bone1Name=dyn_canteen_01, bone2Name=dyn_canteen_MSTR, swing1LimitDegrees=5, swing2LimitDegrees=10, twistLimitDegrees=4}
 {bone1Name=dyn_canteen_02, bone2Name=dyn_canteen_01, swing1LimitDegrees=5, swing2LimitDegrees=20, twistLimitDegrees=4}
 {bone1Name=dyn_canteen_03, bone2Name=dyn_canteen_02, swing1LimitDegrees=0, swing2LimitDegrees=20, twistLimitDegrees=4}
 {bone1Name=dyn_cloth_bk_l_01, bone2Name=dyn_cloth_bk_l_MSTR, swing1LimitDegrees=15, swing2LimitDegrees=30, twistLimitDegrees=0}
 {bone1Name=dyn_cloth_bk_l_02, bone2Name=dyn_cloth_bk_l_01, swing1LimitDegrees=20, swing2LimitDegrees=20, twistLimitDegrees=0}
 {bone1Name=dyn_cloth_bk_l_03, bone2Name=dyn_cloth_bk_l_02, swing1LimitDegrees=15, swing2LimitDegrees=30, twistLimitDegrees=0}
 {bone1Name=dyn_cloth_bk_l_04, bone2Name=dyn_cloth_bk_l_03, swing1LimitDegrees=15, swing2LimitDegrees=30, twistLimitDegrees=0}
 {bone1Name=dyn_cloth_bk_r_01, bone2Name=dyn_cloth_bk_r_MSTR, swing1LimitDegrees=15, swing2LimitDegrees=30, twistLimitDegrees=10}
 {bone1Name=dyn_cloth_bk_r_02, bone2Name=dyn_cloth_bk_r_01, swing1LimitDegrees=20, swing2LimitDegrees=20, twistLimitDegrees=10}
 {bone1Name=dyn_cloth_bk_r_03, bone2Name=dyn_cloth_bk_r_02, swing1LimitDegrees=15, swing2LimitDegrees=30, twistLimitDegrees=10}
 {bone1Name=dyn_cloth_bk_r_04, bone2Name=dyn_cloth_bk_r_03, swing1LimitDegrees=15, swing2LimitDegrees=30, twistLimitDegrees=0}
 {bone1Name=dyn_cloth_fr_l_01, bone2Name=dyn_cloth_fr_l_MSTR, swing1LimitDegrees=15, swing2LimitDegrees=10, twistLimitDegrees=0}
 {bone1Name=dyn_cloth_fr_l_02, bone2Name=dyn_cloth_fr_l_01, swing1LimitDegrees=8, swing2LimitDegrees=10, twistLimitDegrees=5}
 {bone1Name=dyn_cloth_fr_l_03, bone2Name=dyn_cloth_fr_l_02, swing1LimitDegrees=15, swing2LimitDegrees=10, twistLimitDegrees=5}
 {bone1Name=dyn_cloth_fr_l_04, bone2Name=dyn_cloth_fr_l_03, swing1LimitDegrees=15, swing2LimitDegrees=10, twistLimitDegrees=5}
 {bone1Name=dyn_cloth_fr_r_01, bone2Name=dyn_cloth_fr_r_MSTR, swing1LimitDegrees=15, swing2LimitDegrees=10, twistLimitDegrees=5}
 {bone1Name=dyn_cloth_fr_r_02, bone2Name=dyn_cloth_fr_r_01, swing1LimitDegrees=8, swing2LimitDegrees=10, twistLimitDegrees=5}
 {bone1Name=dyn_cloth_fr_r_03, bone2Name=dyn_cloth_fr_r_02, swing1LimitDegrees=15, swing2LimitDegrees=10, twistLimitDegrees=5}
 {bone1Name=dyn_cloth_fr_r_04, bone2Name=dyn_cloth_fr_r_03, swing1LimitDegrees=15, swing2LimitDegrees=10, twistLimitDegrees=5}
 {
  bone1Name=dyn_cloth_fr_r_02
  bone2Name=dyn_cloth_fr_l_02
  swing1LimitDegrees=45
  swing1Motion=Free
  swing2LimitDegrees=45
  swing2Motion=Free
  twistLimitDegrees=45
  twistMotion=Free
 }
 {
  bone1Name=dyn_cloth_fr_r_03
  bone2Name=dyn_cloth_fr_l_03
  swing1LimitDegrees=45
  swing1Motion=Free
  swing2LimitDegrees=45
  swing2Motion=Free
  twistLimitDegrees=45
  twistMotion=Free
 }
 {
  bone1Name=dyn_cloth_fr_r_04
  bone2Name=dyn_cloth_fr_l_04
  swing1LimitDegrees=45
  swing1Motion=Free
  swing2LimitDegrees=45
  swing2Motion=Free
  twistLimitDegrees=45
  twistMotion=Free
 }
 {
  bone1Name=dyn_cloth_bk_l_03
  bone2Name=dyn_cloth_bk_r_03
  swing1LimitDegrees=45
  swing1Motion=Free
  swing2LimitDegrees=45
  swing2Motion=Free
  twistLimitDegrees=45
  twistMotion=Free
 }
 {
  bone1Name=dyn_cloth_bk_l_04
  bone2Name=dyn_cloth_bk_r_04
  swing1LimitDegrees=45
  swing1Motion=Free
  swing2LimitDegrees=45
  swing2Motion=Free
  twistLimitDegrees=45
  twistMotion=Free
 }
]
