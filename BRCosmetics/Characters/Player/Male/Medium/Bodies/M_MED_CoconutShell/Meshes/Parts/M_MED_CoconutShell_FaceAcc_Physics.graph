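# Physics asset for the face accessory part of the male medium CoconutShell player body, setting up a dynamic hair and long braid chain.
# One constrained capsule chain runs from dyn_hair_MSTR through three hair bones and seven braid bones, with tight twist limits and swing generally loosening down the braid.
# Unconstrained capsules cover head, neck, spine, arms and pelvis as body volumes, including extra shapes named tail and hatchet.
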PhysicsAsset /BRCosmetics/Characters/Player/Male/Medium/Bodies/M_MED_CoconutShell/Meshes/Parts/M_MED_CoconutShell_FaceAcc_Physics
bodies {
 clavicle_l [{shapeType=Capsule, length=10.0591, radius=6.5588, shapeName=clavicle_l_capsule, center=<9.1699 0 0>, rotation=<rot 90 0 0>}]
 clavicle_r [{shapeType=Capsule, length=10.7154, radius=6.5875, shapeName=clavicle_r_capsule, center=<-7.8473 0 0>, rotation=<rot 90 0 0>}]
 dyn_braid_01 [{shapeType=Capsule, length=10.5012, radius=2.4659, shapeName=dyn_braid_01_capsule, center=<3.1431 0 0>, rotation=<rot 90 0 0>}]
 dyn_braid_02 [{shapeType=Capsule, length=9.361, radius=1.8201, shapeName=dyn_braid_02_capsule, center=<3.1431 0 0>, rotation=<rot 90 0 0>}]
 dyn_braid_03 [{shapeType=Capsule, length=8.8449, radius=1.714, shapeName=dyn_braid_03_capsule, center=<3.1431 0 0>, rotation=<rot 90 0 0>}]
 dyn_braid_04 [{shapeType=Capsule, length=8.6915, radius=1.3624, shapeName=dyn_braid_04_capsule, center=<3.1431 0 0>, rotation=<rot 90 0 0>}]
 dyn_braid_05 [{shapeType=Capsule, length=9.8646, radius=1.2772, shapeName=dyn_braid_05_capsule, center=<3.1431 0 0>, rotation=<rot 90 0 0>}]
 dyn_braid_06 [{shapeType=Capsule, length=6.5019, radius=1.3131, shapeName=dyn_braid_06_capsule, center=<3.1431 0 0>, rotation=<rot 90 0 0>}]
 dyn_braid_07 [{shapeType=Capsule, length=9.0175, radius=2.2645, shapeName=dyn_braid_07_capsule, center=<5.037 0 0.2378>, rotation=<rot 90 0 0>}]
 dyn_hair_01 [{shapeType=Capsule, length=5.9973, radius=5.2234, shapeName=dyn_hair_01_capsule, center=<4.3061 0 4.9395>, rotation=<rot 84.3189 0 0>}]
 dyn_hair_02 [{shapeType=Capsule, length=10.0527, radius=5.4297, shapeName=dyn_hair_02_capsule, center=<5.4028 0 3.2>, rotation=<rot 82.845 0 0>}]
 dyn_hair_03 [{shapeType=Capsule, length=5.2441, radius=7.1292, shapeName=dyn_hair_03_capsule, center=<5.7866 -0.5094 0.82>, rotation=<rot 82.9872 0 0>}]
 dyn_hair_MSTR [{shapeType=Capsule, length=0.505, radius=0.505, shapeName=dyn_hair_MSTR_capsule, rotation=<rot 0 -90 0>}]
 head [{shapeType=Capsule, length=12.4906, radius=6.2015, shapeName=head_capsule, rotation=<rot 90 0 0>}]
 lowerarm_l [{shapeType=Capsule, length=25.3872, radius=6.7394, shapeName=lowerarm_l_capsule, center=<14.4236 0 0>, rotation=<rot 90 0 0>}]
 lowerarm_r [{shapeType=Capsule, length=25.3872, radius=6.7394, shapeName=lowerarm_r_capsule, center=<-14.799 0 0>, rotation=<rot 90 0 0>}]
 neck_01 [{shapeType=Capsule, length=35.5494, radius=5.4084, shapeName=neck_01_capsule, rotation=<rot 0 -90 0>}]
 neck_02 [{shapeType=Capsule, length=35.5494, radius=5.4084, shapeName=neck_02_capsule, rotation=<rot 0 -90 0>}]
 pelvis [
  {shapeType=Capsule, length=34.1317, radius=3.7941, shapeName=pelvis_capsule, center=<-0.1266 -26.3561 0>, rotation=<rot 0 2.9301 90>}
  {shapeType=Capsule, length=18.8214, radius=12.0578, shapeName=tail_capsule, center=<-3.9674 -0.0747 0>, rotation=<rot 0 -90 0>}
 ]
 spine_01 [
  {shapeType=Capsule, length=35.0184, radius=11.4355, shapeName=spine_01_capsule, rotation=<rot -8.4214 -90 0>}
  {
   shapeType=Capsule
   length=44.7567
   radius=5.7416
   shapeName=hatchet_capsule
   center=<-1.499 -8.1179 0.4732>
   rotation=<rot -13.1849 -90.0002 0>
  }
 ]
 spine_02 [{shapeType=Capsule, length=37.4631, radius=9.8783, shapeName=spine_02_capsule, rotation=<rot 0 -90 0>}]
 spine_03 [{shapeType=Capsule, length=34.2268, radius=9.8783, shapeName=spine_03_capsule, rotation=<rot 0 -90 0>}]
 spine_04 [{shapeType=Capsule, length=32.5225, radius=10.7263, shapeName=spine_04_capsule, center=<0 0.9271 0>, rotation=<rot 0 -90 0>}]
 spine_05 [{shapeType=Capsule, length=32.1961, radius=10.6444, shapeName=spine_05_capsule, rotation=<rot 0 -90 0>}]
 upperarm_l [{shapeType=Capsule, length=25.9175, radius=7.0904, shapeName=upperarm_l_capsule, center=<14.4236 0 0>, rotation=<rot 90 0 0>}]
 upperarm_r [{shapeType=Capsule, length=25.9175, radius=7.0904, shapeName=upperarm_r_capsule, center=<-14.799 0 0>, rotation=<rot 90 0 0>}]
}
constraints [
 {bone1Name=dyn_hair_01, bone2Name=dyn_hair_MSTR, swing1LimitDegrees=4, swing2LimitDegrees=4, twistLimitDegrees=2}
 {bone1Name=dyn_hair_02, bone2Name=dyn_hair_01, swing1LimitDegrees=6, swing2LimitDegrees=6, twistLimitDegrees=2}
 {bone1Name=dyn_hair_03, bone2Name=dyn_hair_02, swing1LimitDegrees=6, swing2LimitDegrees=30, twistLimitDegrees=2}
 {bone1Name=dyn_braid_01, bone2Name=dyn_hair_03, swing1LimitDegrees=6, swing2LimitDegrees=30, twistLimitDegrees=4}
 {bone1Name=dyn_braid_02, bone2Name=dyn_braid_01, swing1LimitDegrees=10, swing2LimitDegrees=20, twistLimitDegrees=4}
 {bone1Name=dyn_braid_03, bone2Name=dyn_braid_02, swing1LimitDegrees=10, swing2LimitDegrees=30, twistLimitDegrees=4}
 {bone1Name=dyn_braid_04, bone2Name=dyn_braid_03, swing1LimitDegrees=20, swing2LimitDegrees=20, twistLimitDegrees=4}
 {bone1Name=dyn_braid_05, bone2Name=dyn_braid_04, swing1LimitDegrees=20, swing2LimitDegrees=20, twistLimitDegrees=4}
 {bone1Name=dyn_braid_06, bone2Name=dyn_braid_05, swing1LimitDegrees=20, swing2LimitDegrees=20, twistLimitDegrees=4}
 {bone1Name=dyn_braid_07, bone2Name=dyn_braid_06, swing1LimitDegrees=20, swing2LimitDegrees=20, twistLimitDegrees=4}
]
