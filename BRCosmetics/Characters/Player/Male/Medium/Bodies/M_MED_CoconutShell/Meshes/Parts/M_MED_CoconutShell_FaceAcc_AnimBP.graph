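# Anim Blueprint for the CoconutShell character's face accessory, animating hair, braids and dreads on a pose copied from the attached parent mesh.
# By default, RigidBody simulation in neck_02 space feeds AnimDynamics on the dread bones.
# With back bling, detected on customization load, it switches to weighted bone constraints, offsets and AnimDynamics on the lower braid with pelvis spherical limits.
# Each update derives gravity and RBAN settings from movement-state presets via a deprecated function.
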AnimBlueprint /BRCosmetics/Characters/Player/Male/Medium/Bodies/M_MED_CoconutShell/Meshes/Parts/M_MED_CoconutShell_FaceAcc_AnimBP
target_skeleton=&/M_MED_CoconutShell_FaceAcc_Skeleton
graphs [
 AnimGraph {nodes [
   n1 = "Transform (Modify) Bone - Bone: dyn_braid_01"(ComponentPose=Local To Component(), Alpha=1, Translation="0.138,0.000,0.550", Rotation="-13.973,6.547,-1.589", Scale="1,1,1") {
    rotationMode=BMM_Additive
    rotationSpace=BCS_BoneSpace
    translationMode=BMM_Additive
    translationSpace=BCS_BoneSpace
    alphaBoolBlend {blendOption=Linear}
    boneToModify {boneName=dyn_braid_01}
    rotation=<rot -13.973 6.547 -1.589>
    translation {x=0.138, z=0.55}
   }
   n2 = "Transform (Modify) Bone - Bone: dyn_braid_02"(ComponentPose=@n1, Alpha=1, Translation="-0.057,0.000,-0.471", Rotation="-16.291,0.000,-0.000", Scale="1,1,1") {
    rotationMode=BMM_Additive
    rotationSpace=BCS_BoneSpace
    translationMode=BMM_Additive
    translationSpace=BCS_BoneSpace
    alphaBoolBlend {blendOption=Linear}
    boneToModify {boneName=dyn_braid_02}
    rotation {pitch=-16.291}
    translation=<-0.057 0 -0.4714>
   }
   n3 = "Transform (Modify) Bone - Bone: dyn_braid_03"(ComponentPose=@n2, Alpha=1, Translation="0,0,0", Rotation="-2.182,0.000,-0.000", Scale="1,1,1") {
    rotationMode=BMM_Additive
    rotationSpace=BCS_BoneSpace
    alphaBoolBlend {blendOption=Linear}
    boneToModify {boneName=dyn_braid_03}
    rotation=<rot -2.1821 0 0>
   }
   n4 = "Transform (Modify) Bone - Bone: dyn_braid_04"(ComponentPose=@n3, Alpha=1, Translation="0,0,0", Rotation="9.268,0.000,-0.000", Scale="1,1,1") {
    rotationMode=BMM_Additive
    rotationSpace=BCS_BoneSpace
    alphaBoolBlend {blendOption=Linear}
    boneToModify {boneName=dyn_braid_04}
    rotation=<rot 9.268 0 0>
   }
   Component To Local(ComponentPose=@n4)
   Output Pose(Result=Use cached pose 'Final Cache'()) {layerGroup=None}
   Save cached pose 'Main Cache'(Pose=Copy Pose From Mesh() {bUseAttachedParent=true}) {cacheName=Main Cache}
   n5 = RigidBody!(
    ComponentPose=Local To Component(LocalPose=Use cached pose 'Main Cache'())
    Alpha
    OverrideWorldGravity:Vector="0,0,0"
    ExternalForce
    SimSpaceSettings:Sim Space Settings Structure=(WorldAlpha=0.7,VelocityScaleZ=1,DampingAlpha=1,MaxLinearVelocity=750,MaxAngularVelocity=5,MaxLinearAcceleration=500,MaxAngularAcceleration=3,ExternalLinearDragV=(X=0,Y=0,Z=0),ExternalLinearVelocity=(X=0,Y=0,Z=0),ExternalAngularVelocity=(X=0,Y=0,Z=0))
   ) -> (Pose) {
    bClampLinearTranslationLimitToRefPose=true
    bOverrideWorldGravity=true
    overridePhysicsAsset=&/M_MED_CoconutShell_FaceAcc_Physics
    simulationSpace=BaseBoneSpace
    alphaBoolBlend {blendOption=Linear}
    baseBoneRef {boneName=neck_02}
    simSpaceSettings {maxAngularAcceleration=3, maxAngularVelocity=5, maxLinearAcceleration=500, maxLinearVelocity=750, worldAlpha=0.7}
   }
   n6 = Anim Dynamics - Bone: dyn_dread_01!(ComponentPose=@n5, Alpha=1, GravityOverride:Vector="0,0,0", ExternalForce="0,0,0") -> (Pose) {
    angularDampingOverride=0.9999
    angularSpringConstant=2000
    bAngularSpring=true
    bOverrideAngularDamping=true
    bUseGravityOverride=true
    simulationSpace=RootRelative
    alphaBoolBlend {blendOption=Linear}
    boundBone {boneName=dyn_dread_01}
    componentAppliedLinearAccClamp=<1000 1000 1000>
    componentLinearAccScale=<0.3 0.3 0.3>
    componentLinearVelScale=<0.4 0.4 0.4>
    physicsBodyDefinitions [{
      boundBone {boneName=dyn_dread_01}
      boxExtents=<20 20 20>
      constraintSetup {angularLimitsMax=<4 20 20>, angularLimitsMin=<-4 -20 0>, angularTarget=<1 0 0>}
      localJointOffset=<0.5 0 0>
     }]
   }
   n7 = Anim Dynamics - Bone: dyn_dread_02!(ComponentPose=@n6, Alpha=1, GravityOverride:Vector="0,0,0", ExternalForce="0,0,0") -> (Pose) {
    angularDampingOverride=0.9999
    angularSpringConstant=2000
    bAngularSpring=true
    bOverrideAngularDamping=true
    bUseGravityOverride=true
    simulationSpace=RootRelative
    alphaBoolBlend {blendOption=Linear}
    boundBone {boneName=dyn_dread_02}
    componentAppliedLinearAccClamp=<1000 1000 1000>
    componentLinearAccScale=<0.3 0.3 0.3>
    componentLinearVelScale=<0.4 0.4 0.4>
    physicsBodyDefinitions [{
      boundBone {boneName=dyn_dread_02}
      boxExtents=<20 20 20>
      constraintSetup {angularLimitsMax=<4 20 0>, angularLimitsMin=<-4 -20 -20>, angularTarget=<1 0 0>}
      localJointOffset=<0.3 0 0>
     }]
   }
   n8 = Anim Dynamics - Bone: dyn_dread_03!(ComponentPose=@n7, Alpha=1, GravityOverride:Vector="0,0,0", ExternalForce="0,0,0") -> (Pose) {
    angularDampingOverride=0.9999
    angularSpringConstant=2000
    bAngularSpring=true
    bOverrideAngularDamping=true
    bUseGravityOverride=true
    simulationSpace=RootRelative
    alphaBoolBlend {blendOption=Linear}
    boundBone {boneName=dyn_dread_03}
    componentAppliedLinearAccClamp=<1000 1000 1000>
    componentLinearAccScale=<0.3 0.3 0.3>
    componentLinearVelScale=<0.4 0.4 0.4>
    physicsBodyDefinitions [{
      boundBone {boneName=dyn_dread_03}
      boxExtents=<20 20 20>
      constraintSetup {angularLimitsMax=<4 20 0>, angularLimitsMin=<-4 -20 -20>, angularTarget=<1 0 0>}
      localJointOffset=<0.4 0 0>
     }]
   }
   n9 = Anim Dynamics - Bone: dyn_dread_04!(ComponentPose=@n8, Alpha=1, GravityOverride:Vector="0,0,0", ExternalForce="0,0,0") -> (Pose) {
    angularDampingOverride=0.9999
    angularSpringConstant=2000
    bAngularSpring=true
    bOverrideAngularDamping=true
    bUseGravityOverride=true
    simulationSpace=RootRelative
    alphaBoolBlend {blendOption=Linear}
    boundBone {boneName=dyn_dread_04}
    componentAppliedLinearAccClamp=<1000 1000 1000>
    componentLinearAccScale=<0.3 0.3 0.3>
    componentLinearVelScale=<0.4 0.4 0.4>
    physicsBodyDefinitions [{
      boundBone {boneName=dyn_dread_04}
      boxExtents=<20 20 20>
      constraintSetup {angularLimitsMax=<4 20 0>, angularLimitsMin=<-4 -20 -20>, angularTarget=<1 0 0>}
      localJointOffset=<0.5 0 0>
     }]
   }
   n10 = Anim Dynamics - Bone: dyn_dread_05!(ComponentPose=@n9, Alpha=1, GravityOverride:Vector="0,0,0", ExternalForce="0,0,0") -> (Pose) {
    angularDampingOverride=0.9999
    angularSpringConstant=2000
    bAngularSpring=true
    bOverrideAngularDamping=true
    bUseGravityOverride=true
    simulationSpace=RootRelative
    alphaBoolBlend {blendOption=Linear}
    boundBone {boneName=dyn_dread_05}
    componentAppliedLinearAccClamp=<1000 1000 1000>
    componentLinearAccScale=<0.3 0.3 0.3>
    componentLinearVelScale=<0.4 0.4 0.4>
    physicsBodyDefinitions [{
      boundBone {boneName=dyn_dread_05}
      boxExtents=<20 20 20>
      constraintSetup {angularLimitsMax=<4 20 20>, angularLimitsMin=<-4 -20 0>, angularTarget=<1 0 0>}
      localJointOffset=<0.4 0 0>
     }]
   }
   n11 = Anim Dynamics - Bone: dyn_dread_06!(ComponentPose=@n10, Alpha=1, GravityOverride:Vector="0,0,0", ExternalForce="0,0,0") -> (Pose) {
    angularDampingOverride=0.9999
    angularSpringConstant=2000
    bAngularSpring=true
    bOverrideAngularDamping=true
    bUseGravityOverride=true
    simulationSpace=RootRelative
    alphaBoolBlend {blendOption=Linear}
    boundBone {boneName=dyn_dread_06}
    componentAppliedLinearAccClamp=<1000 1000 1000>
    componentLinearAccScale=<0.3 0.3 0.3>
    componentLinearVelScale=<0.4 0.4 0.4>
    physicsBodyDefinitions [{
      boundBone {boneName=dyn_dread_06}
      boxExtents=<20 20 20>
      constraintSetup {angularLimitsMax=<4 20 20>, angularLimitsMin=<-4 -20 0>, angularTarget=<1 0 0>}
      localJointOffset=<0.5 0 0>
     }]
   }
   n12 = Anim Dynamics - Bone: dyn_dread_07!(ComponentPose=@n11, Alpha=1, GravityOverride:Vector="0,0,0", ExternalForce="0,0,0") -> (Pose) {
    angularDampingOverride=0.9999
    angularSpringConstant=2000
    bAngularSpring=true
    bOverrideAngularDamping=true
    bUseGravityOverride=true
    simulationSpace=RootRelative
    alphaBoolBlend {blendOption=Linear}
    boundBone {boneName=dyn_dread_07}
    componentAppliedLinearAccClamp=<1000 1000 1000>
    componentLinearAccScale=<0.3 0.3 0.3>
    componentLinearVelScale=<0.4 0.4 0.4>
    physicsBodyDefinitions [{
      boundBone {boneName=dyn_dread_07}
      boxExtents=<20 20 20>
      constraintSetup {angularLimitsMax=<4 20 20>, angularLimitsMin=<-4 -20 0>, angularTarget=<1 0 0>}
      localJointOffset=<0.3 0 0>
     }]
   }
   n13 = Anim Dynamics - Bone: dyn_dread_08!(ComponentPose=@n12, Alpha=1, GravityOverride:Vector="0,0,0", ExternalForce="0,0,0") -> (Pose) {
    angularDampingOverride=0.9999
    angularSpringConstant=2000
    bAngularSpring=true
    bOverrideAngularDamping=true
    bUseGravityOverride=true
    simulationSpace=RootRelative
    alphaBoolBlend {blendOption=Linear}
    boundBone {boneName=dyn_dread_08}
    componentAppliedLinearAccClamp=<1000 1000 1000>
    componentLinearAccScale=<0.3 0.3 0.3>
    componentLinearVelScale=<0.4 0.4 0.4>
    physicsBodyDefinitions [{
      boundBone {boneName=dyn_dread_08}
      boxExtents=<20 20 20>
      constraintSetup {angularLimitsMax=<4 20 20>, angularLimitsMin=<-4 -20 0>, angularTarget=<1 0 0>}
      localJointOffset=<0.5 0 0>
     }]
   }
   Save cached pose 'RBAN Cache'(Pose=Component To Local(ComponentPose=@n13)) {cacheName=RBAN Cache}
   n14 = "Constraint - dyn_hair_01 (PPPP)"!(
    ComponentPose=Local To Component(LocalPose=Use cached pose 'Main Cache'())
    Alpha=1
    ConstraintWeights_0=0.4649
    ConstraintWeights_1:float=0.325
    ConstraintWeights_2:float=0.1305
    ConstraintWeights_3:float=0.0796
   ) -> (Pose) {
    alphaBoolBlend {blendOption=Linear}
    boneToModify {boneName=dyn_hair_01}
    constraintSetup [{targetBone {boneName=head}}, {targetBone {boneName=neck_02}}, {targetBone {boneName=spine_05}}, {targetBone {boneName=neck_01}}]
    constraintWeights [1,1,1,1]
   }
   n15 = "Constraint - dyn_hair_02 (PPPP)"!(
    ComponentPose=@n14
    Alpha=1
    ConstraintWeights_0=0.4649
    ConstraintWeights_1:float=0.325
    ConstraintWeights_2:float=0.1305
    ConstraintWeights_3:float=0.0796
   ) -> (Pose) {
    alphaBoolBlend {blendOption=Linear}
    boneToModify {boneName=dyn_hair_02}
    constraintSetup [{targetBone {boneName=head}}, {targetBone {boneName=neck_02}}, {targetBone {boneName=spine_05}}, {targetBone {boneName=neck_01}}]
    constraintWeights [1,1,1,1]
   }
   n16 = "Constraint - dyn_hair_03 (PPPPPP)"!(
    ComponentPose=@n15
    Alpha=1
    ConstraintWeights_0=0.6444
    ConstraintWeights_1:float=0.1833
    ConstraintWeights_2:float=0.0895
    ConstraintWeights_3:float=0.0439
    ConstraintWeights_4:float=0.0222
    ConstraintWeights_5:float=0.0167
   ) -> (Pose) {
    alphaBoolBlend {blendOption=Linear}
    boneToModify {boneName=dyn_hair_03}
    constraintSetup [
     {targetBone {boneName=spine_05}}
     {targetBone {boneName=neck_02}}
     {targetBone {boneName=neck_01}}
     {targetBone {boneName=spine_03}}
     {targetBone {boneName=clavicle_l}}
     {targetBone {boneName=head}}
    ]
    constraintWeights [1,1,1,1,1,1]
   }
   n17 = "Constraint - dyn_braid_01 (PPPPP)"!(
    ComponentPose=@n16
    Alpha=1
    ConstraintWeights_0=0.7873
    ConstraintWeights_1:float=0.0809
    ConstraintWeights_2:float=0.0645
    ConstraintWeights_3:float=0.042
    ConstraintWeights_4:float=0.0254
   ) -> (Pose) {
    alphaBoolBlend {blendOption=Linear}
    boneToModify {boneName=dyn_braid_01}
    constraintSetup [
     {targetBone {boneName=spine_05}}
     {targetBone {boneName=spine_03}}
     {targetBone {boneName=spine_04}}
     {targetBone {boneName=neck_02}}
     {targetBone {boneName=neck_01}}
    ]
    constraintWeights [1,1,1,1,1]
   }
   n18 = "Constraint - dyn_braid_02 (PPP)"!(ComponentPose=@n17, Alpha=1, ConstraintWeights_0=0.6895, ConstraintWeights_1:float=0.1943, ConstraintWeights_2:float=0.1162) -> (Pose) {
    alphaBoolBlend {blendOption=Linear}
    boneToModify {boneName=dyn_braid_02}
    constraintSetup [{targetBone {boneName=spine_05}}, {targetBone {boneName=spine_04}}, {targetBone {boneName=spine_03}}]
    constraintWeights [1,1,1]
   }
   n19 = "Constraint - dyn_braid_03 (PPPP)"!(
    ComponentPose=@n18
    Alpha=1
    ConstraintWeights_0=0.3894
    ConstraintWeights_1:float=0.3155
    ConstraintWeights_2:float=0.2108
    ConstraintWeights_3:float=0.0843
   ) -> (Pose) {
    alphaBoolBlend {blendOption=Linear}
    boneToModify {boneName=dyn_braid_03}
    constraintSetup [{targetBone {boneName=spine_04}}, {targetBone {boneName=spine_03}}, {targetBone {boneName=spine_05}}, {targetBone {boneName=spine_02}}]
    constraintWeights [1,1,1,1]
   }
   n20 = "Constraint - dyn_braid_04 (PPPPP)"!(
    ComponentPose=@n19
    Alpha=1
    ConstraintWeights_0=0.3693
    ConstraintWeights_1:float=0.317
    ConstraintWeights_2:float=0.1973
    ConstraintWeights_3:float=0.0931
    ConstraintWeights_4:float=0.0233
   ) -> (Pose) {
    alphaBoolBlend {blendOption=Linear}
    boneToModify {boneName=dyn_braid_04}
    constraintSetup [
     {targetBone {boneName=spine_03}}
     {targetBone {boneName=spine_04}}
     {targetBone {boneName=spine_02}}
     {targetBone {boneName=spine_05}}
     {targetBone {boneName=spine_01}}
    ]
    constraintWeights [1,1,1,1,1]
   }
   n21 = "Transform (Modify) Bone - Bone: dyn_hair_01"(ComponentPose=@n20, Alpha=1, Translation="0,0,0", Rotation="0,0,0", Scale="1,1,1") {
    rotationMode=BMM_Additive
    rotationSpace=BCS_BoneSpace
    translationMode=BMM_Additive
    translationSpace=BCS_BoneSpace
    alphaBoolBlend {blendOption=Linear}
    boneToModify {boneName=dyn_hair_01}
   }
   n22 = "Transform (Modify) Bone - Bone: dyn_hair_02"(ComponentPose=@n21, Alpha=1, Translation="0,0,-1.5", Rotation="0,0,0", Scale="1,1,1") {
    rotationMode=BMM_Additive
    rotationSpace=BCS_BoneSpace
    translationMode=BMM_Additive
    translationSpace=BCS_BoneSpace
    alphaBoolBlend {blendOption=Linear}
    boneToModify {boneName=dyn_hair_02}
    translation {z=-1.5}
   }
   n23 = "Transform (Modify) Bone - Bone: dyn_hair_03"(ComponentPose=@n22, Alpha=1, Translation="0,0,-1", Rotation="0,0,0", Scale="1,1,1") {
    rotationMode=BMM_Additive
    rotationSpace=BCS_BoneSpace
    translationMode=BMM_Additive
    translationSpace=BCS_BoneSpace
    alphaBoolBlend {blendOption=Linear}
    boneToModify {boneName=dyn_hair_03}
    translation {z=-1}
   }
   n24 = "Transform (Modify) Bone - Bone: dyn_braid_01"(ComponentPose=@n23, Alpha=1, Translation="0,0,4", Rotation="0,10,0", Scale="1,1,1") {
    rotationMode=BMM_Additive
    rotationSpace=BCS_BoneSpace
    translationMode=BMM_Additive
    translationSpace=BCS_BoneSpace
    alphaBoolBlend {blendOption=Linear}
    boneToModify {boneName=dyn_braid_01}
    rotation {yaw=10}
    translation {z=4}
   }
   n25 = "Transform (Modify) Bone - Bone: dyn_braid_02"(ComponentPose=@n24, Alpha=1, Translation="0,0,0", Rotation="-7.406,0.006,-0.012", Scale="1,1,1") {
    rotationMode=BMM_Additive
    rotationSpace=BCS_BoneSpace
    translationMode=BMM_Additive
    translationSpace=BCS_BoneSpace
    alphaBoolBlend {blendOption=Linear}
    boneToModify {boneName=dyn_braid_02}
    rotation=<rot -7.4061 0.006 -0.012>
   }
   n26 = "Transform (Modify) Bone - Bone: dyn_braid_03"(ComponentPose=@n25, Alpha=1, Translation="0,0,0", Rotation="5.810,-0.008,-0.012", Scale="1,1,1") {
    rotationMode=BMM_Additive
    rotationSpace=BCS_BoneSpace
    translationMode=BMM_Additive
    translationSpace=BCS_BoneSpace
    alphaBoolBlend {blendOption=Linear}
    boneToModify {boneName=dyn_braid_03}
    rotation=<rot 5.81 -0.008 -0.012>
   }
   n27 = "Transform (Modify) Bone - Bone: dyn_braid_04"(ComponentPose=@n26, Alpha=1, Translation="0,0,0", Rotation="2.328,0.855,0.118", Scale="1,1,1") {
    rotationMode=BMM_Additive
    rotationSpace=BCS_BoneSpace
    translationMode=BMM_Additive
    translationSpace=BCS_BoneSpace
    alphaBoolBlend {blendOption=Linear}
    boneToModify {boneName=dyn_braid_04}
    rotation=<rot 2.328 0.855 0.118>
   }
   n28 = Anim Dynamics - Bone: dyn_braid_05!(ComponentPose=@n27, Alpha=1, GravityOverride:Vector="0,0,-980", ExternalForce="0,0,0") -> (Pose) {
    angularDampingOverride=1
    angularSpringConstant=3000
    bAngularSpring=true
    bOverrideAngularDamping=true
    bUseGravityOverride=true
    bUseSphericalLimits=true
    gravityScale=0.35
    alphaBoolBlend {blendInTime=2, blendOption=Cubic}
    boundBone {boneName=dyn_braid_05}
    componentAppliedLinearAccClamp=<1000 1000 1000>
    componentLinearAccScale=<0.3 0.3 0.3>
    componentLinearVelScale=<0.4 0.4 0.4>
    physicsBodyDefinitions [{
      boundBone {boneName=dyn_braid_05}
      boxExtents=<16 16 16>
      constraintSetup {angularLimitsMax=<0 25 5>, angularLimitsMin=<0 -10 -5>, angularTarget=<1 0 0>}
      localJointOffset=<1 0 0>
     }]
    relativeSpaceBone {boneName=root}
    sphericalLimits [{limitRadius=15, drivingBone {boneName=pelvis}, sphereLocalOffset=<0 0 0>}]
   }
   n29 = Anim Dynamics - Bone: dyn_braid_06!(ComponentPose=@n28, Alpha=1, GravityOverride:Vector="0,0,-980", ExternalForce="0,0,0") -> (Pose) {
    angularDampingOverride=1
    angularSpringConstant=3000
    bAngularSpring=true
    bOverrideAngularDamping=true
    bUseGravityOverride=true
    bUseSphericalLimits=true
    gravityScale=0.35
    alphaBoolBlend {blendInTime=2, blendOption=Cubic}
    boundBone {boneName=dyn_braid_06}
    componentAppliedLinearAccClamp=<1000 1000 1000>
    componentLinearAccScale=<0.3 0.3 0.3>
    componentLinearVelScale=<0.4 0.4 0.4>
    physicsBodyDefinitions [{
      boundBone {boneName=dyn_braid_06}
      boxExtents=<16 16 16>
      constraintSetup {angularLimitsMax=<0 25 5>, angularLimitsMin=<0 -10 -5>, angularTarget=<1 0 0>}
      localJointOffset=<1 0 0>
     }]
    relativeSpaceBone {boneName=root}
    sphericalLimits [{limitRadius=15, drivingBone {boneName=pelvis}, sphereLocalOffset=<0 0 0>}]
   }
   n30 = Anim Dynamics - Bone: dyn_braid_07!(ComponentPose=@n29, Alpha=1, GravityOverride:Vector="0,0,-980", ExternalForce="0,0,0") -> (Pose) {
    angularDampingOverride=1
    angularSpringConstant=3000
    bAngularSpring=true
    bOverrideAngularDamping=true
    bUseGravityOverride=true
    bUseSphericalLimits=true
    gravityScale=0.35
    alphaBoolBlend {blendInTime=2, blendOption=Cubic}
    boundBone {boneName=dyn_braid_07}
    componentAppliedLinearAccClamp=<1000 1000 1000>
    componentLinearAccScale=<0.3 0.3 0.3>
    componentLinearVelScale=<0.4 0.4 0.4>
    physicsBodyDefinitions [{
      boundBone {boneName=dyn_braid_07}
      boxExtents=<16 16 16>
      constraintSetup {angularLimitsMax=<0 25 5>, angularLimitsMin=<0 -10 -5>, angularTarget=<1 0 0>}
      localJointOffset=<1 0 0>
     }]
    relativeSpaceBone {boneName=root}
    sphericalLimits [{limitRadius=15, drivingBone {boneName=pelvis}, sphereLocalOffset=<0 0 0>}]
   }
   Save cached pose 'Backpack Cache'(Pose=Component To Local(ComponentPose=@n30)) {cacheName=Backpack Cache}
   Save cached pose 'Final Cache'(Pose=Blend Poses by bool(
     BlendPose_0=Use cached pose 'Backpack Cache'()
     BlendPose_1=Use cached pose 'RBAN Cache'()
     bActiveValue=Get isWearingPack() -> (isWearingPack:bool)
    ) {blendType=Linear}) {cacheName=Final Cache}
  ]}
 EventGraph {nodes [
   Event BlueprintUpdateAnimation()
   > n1 = Gravity Override State Function DEPRECATED(
    CharacterPartAnimInstance=Self-Reference()
    GameTimeInSeconds=GetGameTimeInSeconds()
    GravityOverrideParams=Get Gravity Override Params()
    RBAN Params=Get RBAN Params()
   )
   > Set Gravity Override(Gravity Override=@n1)
   > Set RBAN Settings Out(RBAN Settings Out:RBANSetting Struct Structure=@n1.RBAN SettingsOut) -> (Output_Get:RBANSetting Struct Structure)
   Event OnCharacterCustomizationLoaded()
   n2 = CheckForBackbling()
   > Set isWearingPack(isWearingPack:bool=OR Boolean!(A=@n2, B=@n2.bIsBackblingACape, C:bool=@n2.bIsBackblingLong) -> (ReturnValue)) -> (Output_Get:bool)
   > Set isWearingCape(isWearingCape=@n2.bIsBackblingLong)
  ]}
]
variables [
 Gravity Override {type=Vector}
 Gravity Override Params {
  type=GravityOverrideParamsStruct
  default {
   onGround_Moving {gravityOverride=<0 -100 -600>}
   skydive_DivingDown {gravityOverride=<-1500 -600 0>}
   skydive_FloatingDown {gravityOverride=<0 -300 3000>}
   skydive_Parachuting {gravityOverride=<0 -800 -200>}
   swimming {gravityOverride=<-800 0 -100>}
   tactical_Sprint {gravityOverride=<0 -800 -600>}
   ziplining {gravityOverride=<0 -600 -400>}
  }
 }
 Gravity Override Settings Out {
  type=GravityOverrideSettingStruct
  default {jointName=root, gravityOverride=<0 0 -980>, windAmplitude=<0 0 0>, windFrequency=<0 0 0>, windOffset=<0 0 0>}
 }
 isWearingCape {type=boolean}
 isWearingPack {type=boolean}
 RBAN Params {
  type=RBANParamsStruct
  default {
   dBNO {simSpaceSettings {
     dampingAlpha=0.5
     maxAngularAcceleration=20
     maxAngularVelocity=4
     maxLinearAcceleration=1000
     maxLinearVelocity=200
     velocityScaleZ=0.4
     externalLinearDragV=<0.5 0.5 0.7>
    }}
   dBNO_Carried {simSpaceSettings {
     dampingAlpha=0.5
     maxAngularAcceleration=20
     maxAngularVelocity=4
     maxLinearAcceleration=1000
     maxLinearVelocity=200
     velocityScaleZ=0.4
     externalLinearDragV=<0.5 0.5 0.7>
    }}
   emoteOrMelee {simSpaceSettings {
     maxAngularAcceleration=0.1
     maxAngularVelocity=0.25
     maxLinearAcceleration=2.5
     maxLinearVelocity=5
     velocityScaleZ=1
     externalLinearDragV=<0 0 0>
    }}
   front_End {simSpaceSettings {
     maxAngularAcceleration=3
     maxAngularVelocity=5
     maxLinearAcceleration=500
     maxLinearVelocity=750
     velocityScaleZ=1
     externalLinearDragV=<0 0 0>
    }}
   onGround_Crouch_Moving {simSpaceSettings {
     dampingAlpha=0.5
     maxAngularAcceleration=20
     maxAngularVelocity=4
     maxLinearAcceleration=1000
     maxLinearVelocity=200
     velocityScaleZ=0.4
     externalLinearDragV=<0.5 0.5 0.7>
    }}
   onGround_Falling {simSpaceSettings {
     dampingAlpha=0.5
     maxAngularAcceleration=20
     maxAngularVelocity=4
     maxLinearAcceleration=1000
     maxLinearVelocity=200
     velocityScaleZ=0.4
     externalLinearDragV=<0.5 0.5 0.7>
    }}
   onGround_Moving {
    componentLinearAccScale=<0 0 0>
    componentLinearVelScale=<0 0 0>
    simSpaceSettings {
     dampingAlpha=0.5
     maxAngularAcceleration=20
     maxAngularVelocity=4
     maxLinearAcceleration=1000
     maxLinearVelocity=200
     velocityScaleZ=0.4
     externalLinearDragV=<0.5 0.5 0.7>
    }
   }
   onGround_Standing {
    componentLinearAccScale=<0 0 0>
    componentLinearVelScale=<0 0 0>
    simSpaceSettings {
     dampingAlpha=0.5
     maxAngularAcceleration=20
     maxAngularVelocity=4
     maxLinearAcceleration=1000
     maxLinearVelocity=200
     velocityScaleZ=0.4
     externalLinearDragV=<0.5 0.5 0.7>
    }
   }
   rail_Grinding {simSpaceSettings {
     dampingAlpha=0.5
     maxAngularAcceleration=20
     maxAngularVelocity=4
     maxLinearAcceleration=1000
     maxLinearVelocity=200
     velocityScaleZ=0.4
     worldAlpha=1
     externalLinearDragV=<0.5 0.5 0.7>
    }}
   riding_Moving {simSpaceSettings {
     dampingAlpha=0.5
     maxAngularAcceleration=20
     maxAngularVelocity=4
     maxLinearAcceleration=1000
     maxLinearVelocity=200
     velocityScaleZ=0.4
     externalLinearDragV=<0.5 0.5 0.7>
    }}
   riding_Moving_Targeting {simSpaceSettings {
     dampingAlpha=0.5
     maxAngularAcceleration=20
     maxAngularVelocity=4
     maxLinearAcceleration=1000
     maxLinearVelocity=200
     velocityScaleZ=0.4
     externalLinearDragV=<0.5 0.5 0.7>
    }}
   riding_Standing {simSpaceSettings {
     dampingAlpha=0.5
     maxAngularAcceleration=20
     maxAngularVelocity=4
     maxLinearAcceleration=1000
     maxLinearVelocity=200
     velocityScaleZ=0.4
     externalLinearDragV=<0.5 0.5 0.7>
    }}
   skydive_DivingDown {
    componentLinearAccScale=<0 0 0>
    componentLinearVelScale=<-3 0 0>
    simSpaceSettings {
     dampingAlpha=0.5
     maxAngularAcceleration=20
     maxAngularVelocity=4
     maxLinearAcceleration=1000
     maxLinearVelocity=200
     velocityScaleZ=0.4
     externalLinearDragV=<0.5 0.5 0.7>
    }
   }
   skydive_FloatingDown {
    componentLinearAccScale=<0 0 0>
    componentLinearVelScale=<0 0 0>
    simSpaceSettings {
     dampingAlpha=0.5
     maxAngularAcceleration=20
     maxAngularVelocity=4
     maxLinearAcceleration=1000
     maxLinearVelocity=200
     velocityScaleZ=0.4
     externalLinearDragV=<0.5 0.5 0.7>
    }
   }
   skydive_GoingUp {simSpaceSettings {
     dampingAlpha=0.5
     maxAngularAcceleration=20
     maxAngularVelocity=4
     maxLinearAcceleration=1000
     maxLinearVelocity=200
     velocityScaleZ=0.4
     externalLinearDragV=<0.5 0.5 0.7>
    }}
   skydive_Parachuting {
    componentLinearAccScale=<0 0 0>
    componentLinearVelScale=<0 0 0>
    simSpaceSettings {
     maxAngularAcceleration=3
     maxAngularVelocity=5
     maxLinearAcceleration=500
     maxLinearVelocity=750
     velocityScaleZ=1
     externalLinearDragV=<0 0 0>
    }
   }
   skydive_Surfing {simSpaceSettings {
     dampingAlpha=0.5
     maxAngularAcceleration=20
     maxAngularVelocity=4
     maxLinearAcceleration=1000
     maxLinearVelocity=200
     velocityScaleZ=0.4
     externalLinearDragV=<0.5 0.5 0.7>
    }}
   swimming {simSpaceSettings {
     maxAngularAcceleration=0.5
     maxAngularVelocity=1
     maxLinearAcceleration=75
     maxLinearVelocity=150
     velocityScaleZ=1
     externalLinearDragV=<0 0 0>
    }}
   tactical_Sprint {
    componentLinearAccScale=<0 0 0>
    componentLinearVelScale=<0 0 0>
    simSpaceSettings {
     dampingAlpha=0.5
     maxAngularAcceleration=20
     maxAngularVelocity=4
     maxLinearAcceleration=1000
     maxLinearVelocity=200
     velocityScaleZ=0.4
     externalLinearDragV=<0.5 0.5 0.7>
    }
   }
   ziplining {simSpaceSettings {
     dampingAlpha=0.5
     maxAngularAcceleration=20
     maxAngularVelocity=4
     maxLinearAcceleration=1000
     maxLinearVelocity=200
     velocityScaleZ=0.4
     externalLinearDragV=<0.5 0.5 0.7>
    }}
  }
 }
 RBAN Settings Out {
  type=RBANSettingStruct
  default {
   componentLinearAccScale=<0 0 0>
   componentLinearVelScale=<0 0 0>
   simSpaceSettings {
    dampingAlpha=1
    maxAngularAcceleration=1000
    maxAngularVelocity=3
    maxLinearAcceleration=300
    maxLinearVelocity=700
    velocityScaleZ=0
    worldAlpha=1
    externalAngularVelocity=<0 0 0>
    externalLinearDragV=<0 0.5 0>
    externalLinearVelocity=<0 0 0>
   }
  }
 }
]
